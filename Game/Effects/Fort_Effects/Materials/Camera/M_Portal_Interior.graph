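# Unlit, dithered masked mesh-particle material for a portal interior: deep space seen behind cracking glass.
# Dynamic parameters Cracking and Open drive crack strength and slide crack UVs, which panning chromatic smoke distorts.
# Emissive layers a crack-bent skylight cubemap reflection, camera-vector triplanar space texture and particle-tinted crack glow, with rims from offset mask subtraction.
# Crack mask, particle alpha and camera-distance cutoff form a mask; mobile drops smoke distortion.
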Material /Game/Effects/Fort_Effects/Materials/Camera/M_Portal_Interior
graph {
 properties {
  bUsedWithMeshParticles=true
  bUsedWithNiagaraMeshParticles=true
  blendMode=BLEND_Masked
  ditherOpacityMask=true
  pixelDepthOffsetMode=PDOM_Legacy
  shadingModel=MSM_Unlit
  usedShadingModels=Unlit
 }
 nodes [
  MaterialFunctionCall() {materialFunction=&DebugTimeSine}
  n1 = DynamicParameter() {paramNames [Open,Cracking,Param3,Param4]}
  n2 = Constant3Vector() {constant=<rgba 0.5521 0.5154 1 0>}
  n3 = TextureSample(UVs=Panner(Coordinate=TextureCoordinate() {uTiling=4, vTiling=0.4}) {speedX=-0.2, speedY=1}) {texture=&T_Inky_Smoke_Tile_Chromatic}
  n4 = Panner(Coordinate=BumpOffset(Coordinate=TextureCoordinate() {uTiling=3, vTiling=0.4}, Height=Constant() {r=10})) {speedX=0.2, speedY=1}
  n5 = Multiply(
   A=LinearInterpolate(
    A=Multiply(A=@n2, B=@n3.RGB)
    B=TextureSample(UVs=@n4) {texture=&T_Inky_Smoke_Tile_Chromatic}.RGB
   )
   B=@n1.Cracking
  )
  n6 = ComponentMask(None=ShadingPathSwitch(Default=@n5, Mobile=Constant()) {desc=Migrated from FeatureLevelSwitch}) {g=true, r=true}
  n7 = Constant2Vector() {g=-0.1, r=-0.02}
  n8 = Add(
   A=TextureCoordinate() {coordinateIndex=1, uTiling=5, vTiling=1.7}
   B=AppendVector(A=Constant(), B=LinearInterpolate(Alpha=@n1.Open) {constA=-1.15, constB=-0.65})
  )
  n9 = TextureSample(UVs=Add(A=Multiply(A=@n6, B=@n7), B=@n8)) {texture=/Game/Effects/Fort_Effects/Textures/Tile/T_Glass_Cracks_M}
  n10 = Multiply(A=@n1.Cracking, B=@n9.R)
  n11 = ReflectionVectorWS(CustomWorldNormal=AppendVector(A=LinearInterpolate(Alpha=@n10) {constA=-0.2, constB=0.2}, B=Constant2Vector() {g=1})) {bNormalizeCustomWorldNormal=true}
  n12 = ComponentMask(None=TextureCoordinate()) {g=true}
  n13 = ShadingPathSwitch(Default=Multiply(A=@n12, B=Multiply(A=@n5) {constB=0.5}) {constB=0.5}, Mobile=Constant()) {desc=Migrated from FeatureLevelSwitch}
  n14 = Multiply(
   A=TextureSampleParameterCube!(UVs=Add(A=@n11, B=@n13), Apply View MipBias) {samplerType=SAMPLERTYPE_LinearColor, texture=/Game/Packages/Fortress_Sky/TexturesHDR/T_AthenaSkylight}.RGB
   B=VectorParameter() {parameterName=CubemapTint, defaultValue=<rgba 0.068 0.0531 0.15 0>}.RGB
  )
  n15 = MaterialFunctionCall!(
   CameraVector
   Texture=TextureObject() {texture=/Game/Effects/Fort_Effects/Textures/Tile/T_Portal_Space_D}
   Tiling=Constant() {r=1}
   Offset=Multiply(A=@n10) {constB=-1}
   AxisFadeContrast=Constant() {r=1}
  ) {materialFunction=/Engine/Functions/Engine_MaterialFunctions01/Texturing/TriplanarCameraVector}
  n16 = ParticleColor()
  n17 = ShadingPathSwitch(Default=@n12, Mobile=Constant()) {desc=Migrated from FeatureLevelSwitch}
  n18 = Multiply(A=@n9.G) {constB=20}
  n19 = Power(Base=ComponentMask(None=TextureCoordinate()) {g=true}) {constExponent=7}
  n20 = Add(
   A=Add(A=Multiply(A=@n5) {constB=1.5}, B=Multiply(A=@n17, B=@n18))
   B=Multiply(A=@n19, B=ScalarParameter() {defaultValue=2, parameterName=GradientGlow}) {constB=0.25}
  )
  n21 = MaterialFunctionCall!(Use Alternate UVs, Alternate UVs) {materialFunction=/Game/Athena/Effects/Materials/Geode/Functions/MF_Geode_Crack_Colors}
  n22 = TextureSample(UVs=@n8) {texture=/Game/Effects/Fort_Effects/Textures/Tile/T_Glass_Cracks_Mask}
  n23 = Constant2Vector() {g=0.005}
  n24 = Subtract(A=@n22.G, B=TextureSample(UVs=Add(A=@n8, B=@n23)) {texture=&T_Glass_Cracks_Mask}.G)
  n25 = Add(A=Multiply(A=Multiply(A=@n20, B=@n19), B=@n21.Crack Color), B=Multiply(A=Multiply(A=@n21.Crack Color) {constB=3}, B=Clamp(None=@n24)))
  MaterialFunctionCall(Emissive=Add(A=Add(A=@n14, B=@n15.Result), B=Multiply(A=@n16.RGB, B=@n25))) {materialFunction=&MF_FX_EmissiveAdjustments}
  n26 = Add(A=Multiply(A=@n10) {constB=0.25})
  n27 = ScalarParameter() {defaultValue=1500, parameterName=HideDistance}
  n28 = ObjectPositionWS()
  n29 = ActorPositionWS()
  n30 = ScalarParameter() {parameterName=Use actorPos}
  n31 = MaterialFunctionCall(Vector 3=Subtract(A=LinearInterpolate(A=@n28, B=@n29, Alpha=@n30), B=CameraPositionWS())) {materialFunction=/Engine/Functions/Engine_MaterialFunctions02/Utility/VectorLength}
  Multiply(A=Multiply(A=@n16.A, B=Multiply(A=@n26, B=@n22.G)), B=Clamp(None=Subtract(A=@n27, B=@n31.V3 Length)))
  Multiply(A=@n15.Result) {constB=4}
  n32 = Multiply(A=VertexColor().R) {constB=1.2}
  n33 = TextureCoordinate()
  n34 = Multiply(A=Power(Base=ComponentMask(None=@n33) {g=true}) {constExponent=8}) {constB=0.7}
  n35 = Subtract(A=WorldPosition().XYZ, B=ObjectPositionWS())
  n36 = VectorParameter() {parameterName=WPOMult, defaultValue=<rgba 8 8 0 0>}
  Multiply(A=LinearInterpolate(Alpha=@n1.Open) {constA=-0.14}, B=Multiply(A=Add(A=@n32, B=@n34), B=Multiply(A=@n35, B=@n36.RGB))) {constB=0}
  n37 = Multiply(A=CameraVectorWS())
  n38 = Add(A=Multiply(A=ScalarParameter() {defaultValue=0.2, group=Galaxy, parameterName=Galaxy Rotation Speed_2}, B=Time()))
  TextureSample(UVs=ComponentMask(None=Add(A=@n37, B=@n38)) {g=true, r=true}) {samplerSource=SSM_Wrap_WorldGroupSettings, texture=&T_Portal_Space_D}
 ]
}
